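# Niagara system for the SpaceWanderer skin that puts two small colored glows with lights on the character's skeletal mesh.
# One emitter bursts two particles, placed at spawn on bones or sockets indexed by spawn group.
# Each update re-samples the mesh location using particle ID modulo 2, keeping the glows attached.
# Sprite and light renderers share the particles, with color from User.ParticleColor and light radius derived from sprite size.
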NiagaraSystem /BRCosmetics/Athena/Effects/Particles/Skins/SpaceWanderer/NS_Skin_SpaceWanderer_Lights
dependencies {
 usedDataInterfaces [&NiagaraDataInterfaceCurve]
 usedDynamicInputs []
 usedModules [/Niagara/Modules/System/SystemLifeCycle, &SystemState]
}
emitters {FN_Sprite {
  emitterUpdateScript [
   &EmitterScalability(
    Module.bApplySpawnCountScale=false
    Module.bApplySpawnCountScaleByDistanceFraction=false
    Module.bUseMaxDistance=false
    Module.bUseMinDistance=false
   ) {enabled=false}
   &EmitterLifeCycle(NextLoopDuration=1.0042) {enabled=false}
   &EmitterState()
   &SpawnBurst_Instantaneous(Spawn Count=2, Use Spawn Probability=false)
  ]
  particleSpawnScript [
   /Game/Effects/Niagara/Modules/Groups/FN_SetSpawnGroup_v1/FN_SetSpawnGroup(Emitter Bone Count=2)
   /Niagara/Modules/Spawn/Initialization/InitializeParticle(
    Lifetime=0.5
    Module.Write Color=true
    Module.Write Lifetime=true
    Module.Write Mass=true
    Module.Write Material Random=true
    Module.Write Position=false
    Module.Write Scale=false
    Module.Write SpriteRotation=true
    Module.Write SpriteSize=true
    Position=&SimulationPosition()
    Sprite Rotation=&UniformRangedFloat(Minimum=0, Maximum=360)
    Sprite Size=&Vector2DFromFloat(Value=5)
   )
   /SetVariables(Particles.Color=$User.ParticleColor) {isSetParameters=true}
   /SetVariables(Particles.LightRadius=/Niagara/DynamicInputs/TypeConversions/Vector2DLength(Vector:Vector2f=$Particles.SpriteSize)) {isSetParameters=true}
   /SetVariables(Particles.LightExponent=2) {isSetParameters=true}
   &DynamicMaterialParameters(
    Index 0 Param 1=2
    Module.Param0WriteEnabled=true
    Module.Param1WriteEnabled=true
    Module.Param2WriteEnabled=true
    Module.Param3WriteEnabled=true
   )
   &SampleSkeletalMeshSkeleton(
    Mesh:NiagaraDataInterfaceSkeletalMesh=$User.SkeletalMesh
    Spawn Mode:ENiagaraSkeletalSpawning=::NewEnumerator1
    Specific Bone Index:NiagaraInt32=$Particles.SpawnGroup
    Specific Socket Index:NiagaraInt32=$Particles.SpawnGroup
   )
   /Niagara/Modules/Spawn/Location/SkeletalMeshSkeletonLocation(Module.Inherit Velocity=false, Module.Use Sampled Bone as Ribbon ID=true)
  ]
  particleUpdateScript [
   &SkeletalMeshLocation(
    Bone / Socket Index=&MakeIntFromFloat(Float=&ModuloFloat(A=&MakeFloatFromInt(Int=/Niagara/DynamicInputs/Execution/ReturnParticleID()), B=2))
    Inherit Velocity=false
    MaskSpawnGroup=false
    Mesh Sampling Type=::NewEnumerator5
    Skeletal Mesh=$User.SkeletalMesh
    Socket Sampling Mode=::NewEnumerator5
   )
   &ParticleState()
   &SolveForcesAndVelocity(Clamp Velocity=false, Limit Acceleration=false) {enabled=false}
  ]
  rendererClasses [&NiagaraLightRendererProperties, &NiagaraSpriteRendererProperties]
  renderers [
   {rendererClass=&NiagaraLightRendererProperties, rendererIndex=0}
   {rendererClass=&NiagaraSpriteRendererProperties, rendererIndex=1}
  ]
 }}
summary {
 systemName=NS_Skin_SpaceWanderer_Lights
 emitters [{emitterName=FN_Sprite, rendererClasses [&NiagaraLightRendererProperties, &NiagaraSpriteRendererProperties]}]
 userVariables [
  User.bIsFrontEndPreview {type=NiagaraBool}
  User.GlobalOffset {type=Vector3f, description="", defaultValue {struct=Vector3f, value=<0 0 0>}}
  User.ParticleColor {type=LinearColor, description="", defaultValue {struct=LinearColor, value=<rgba 1 1 1 1>}}
  User.PlayerSpeed {type=NiagaraFloat, defaultValue=0, description=""}
  User.RibbonIDOffset {type=NiagaraInt32}
  User.SkeletalMesh {
   type=NiagaraDataInterfaceSkeletalMesh
   defaultValue {
    struct=NiagaraExt_VariableValue_DataInterface
    value {dataInterface=&self:NiagaraDataInterfaceSkeletalMesh_0}
   }
  }
 ]
}
